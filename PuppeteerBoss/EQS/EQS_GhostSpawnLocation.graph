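# EQS query that picks a ghost spawn location around a suitable player for the Puppeteer boss.
# A spiral donut of points around that player projects onto navigation using a no-water filter.
# Tests check for flat ground and favour points obstructed relative to the player and farther away.
# A dot test against player facing keeps only points behind the player, scoring those most directly behind highest.
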EnvQuery /PuppeteerBoss/EQS/EQS_GhostSpawnLocation
options [{
  generator {
   name=EnvQueryGenerator_Donut_0
   class=EnvQueryGenerator_Donut
   properties {
    bUseSpiralPattern=true
    center=&/FortQueryContext_BP_GhostSuitablePlayer.FortQueryContext_BP_GhostSuitablePlayer_C
    innerRadius=1000
    numberOfRings=3
    outerRadius=2000
    pointsPerRing=8
    projectionData {
     extentX=0
     navigationFilter=/Game/Athena/AI/Phoebe/NavFilter/NavFilter_Phoebe_NoWater.NavFilter_Phoebe_NoWater_C
     postProjectionVerticalOffset=0
    }
   }
  }
  tests [
   FortQueryTest_IsObstructed_0 {
    class=FortQueryTest_IsObstructed
    properties {
     boolValue=true
     context=&FortQueryContext_BP_GhostSuitablePlayer_C
     testPurpose=Score
    }
   }
   FortQueryTest_OnFlatSurface_0 {class=FortQueryTest_OnFlatSurface, properties {radius=50, toleranceZ=10, traceOffsetUp=100}}
   EnvQueryTest_Distance_0 {
    class=EnvQueryTest_Distance
    properties {
     distanceTo=&FortQueryContext_BP_GhostSuitablePlayer_C
     filterType=Range
     floatValueMin=0
     scoringEquation=Linear
     scoringFactor=1
     testComment=""
     testMode=Distance3D
     testPurpose=Score
    }
   }
   EnvQueryTest_Dot_0 {
    class=EnvQueryTest_Dot
    properties {
     clampMaxType=None
     clampMinType=None
     filterType=Maximum
     floatValueMax=-0.5
     floatValueMin=-0.5
     scoreClampMax=0
     scoreClampMin=0
     scoringEquation=InverseLinear
     scoringFactor=1
     testComment=""
     testMode=Dot3D
     testPurpose=FilterAndScore
     lineA {rotation=&FortQueryContext_BP_GhostSuitablePlayer_C}
     lineB {lineFrom=&FortQueryContext_BP_GhostSuitablePlayer_C}
    }
   }
  ]
 }]
