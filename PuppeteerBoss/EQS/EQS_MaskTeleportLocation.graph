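# EQS query that picks a teleport location for a Mask in the Puppeteer boss content.
# Builds a point grid in a box defined by the teleport area center and extent contexts, projected onto the navmesh with a no-water filter.
# Distance tests discard points too close in 2D to the Mask's player target or to the querier.
# A random test scores the remaining points, so the destination varies within the valid area.
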EnvQuery /PuppeteerBoss/EQS/EQS_MaskTeleportLocation
options [{
  generator {
   name=FortQueryGenerator_GridInBox_0
   class=FortQueryGenerator_GridInBox
   properties {
    bAutoSortTests=true
    bCanRunAsync=false
    boxExtentsContext=&/FortQueryContext_BP_MaskTeleportAreaExtent.FortQueryContext_BP_MaskTeleportAreaExtent_C
    boxLength=1000
    boxWidth=1000
    boxesContext=null
    envQueryResultNormalizationOption=Default
    generateAround=&/FortQueryContext_BP_MaskTeleportAreaCenter.FortQueryContext_BP_MaskTeleportAreaCenter_C
    minimumPointsOnLength=0
    minimumPointsOnWidth=0
    navDataOverrideContext=null
    optionName=""
    spaceBetween=300
    projectionData {
     bCanDisableTrace=true
     bCanProjectDown=true
     bCanTraceOnGeometry=true
     bCanTraceOnNavMesh=true
     bOnlyBlockingHits=true
     bTraceComplex=false
     extentX=0
     extentY=10
     extentZ=10
     navigationFilter=/Game/Athena/AI/Phoebe/NavFilter/NavFilter_Phoebe_NoWater.NavFilter_Phoebe_NoWater_C
     postProjectionVerticalOffset=0
     projectDown=1024
     projectUp=1024
     serializedChannel=ECC_Visibility
     traceChannel=TraceTypeQuery1
     traceMode=Navigation
     traceProfileName=None
     traceShape=Line
     versionNum=1
    }
   }
  }
  tests [
   EnvQueryTest_Distance_2 {
    class=EnvQueryTest_Distance
    properties {
     distanceTo=&/FortQueryContext_BP_MaskPlayerTarget.FortQueryContext_BP_MaskPlayerTarget_C
     filterType=Minimum
     floatValueMax=100
     floatValueMin=800
     scoringEquation=Linear
     scoringFactor=1
     testComment=""
     testMode=Distance2D
     testPurpose=Filter
    }
   }
   EnvQueryTest_Distance_1 {
    class=EnvQueryTest_Distance
    properties {
     distanceTo=&EnvQueryContext_Querier
     filterType=Minimum
     floatValueMax=100
     floatValueMin=500
     scoringEquation=Linear
     scoringFactor=1
     testComment=""
     testMode=Distance2D
     testPurpose=Filter
    }
   }
   FortQueryTest_Random_0 {class=FortQueryTest_Random, properties {floatValueMax=0, scoringFactor=1}}
  ]
 }]
