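; EQS query context for the Puppeteer boss that supplies the ghost manager's suitable player as a single reference actor.
; It casts the querier to GhostManager_BP and returns its SuitablePlayer if valid.
; When the cast fails, it falls back to a FirePetalBoss_PlayerTargetEQSDummy actor found in the level, so queries can resolve for other queriers.
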(fn ProvideSingleActor
  (params (QuerierObject :type Object%20Reference) (QuerierActor :type Actor%20Object%20Reference))
  (bind _asghost_manager_bp (Utilities|Casting|CastToGhostManager_BP QuerierActor)
    (:then
      (Utilities|IsValid (Variables|/PuppeteerBoss/Manager/GhostManager_BP.GhostManager_BP_C|GetSuitablePlayer _asghost_manager_bp)
        (:Is%20Valid
          (return (Variables|/PuppeteerBoss/Manager/GhostManager_BP.GhostManager_BP_C|GetSuitablePlayer _asghost_manager_bp)))))
    (:CastFailed
      (bind _returnvalue (Actor|GetActorOfClass "/FirePetalBosses/EQS/FirePetalBoss_PlayerTargetEQSDummy.FirePetalBoss_PlayerTargetEQSDummy_C"))
      (Utilities|IsValid _returnvalue
        (:Is%20Valid
          (return _returnvalue))))))
