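; EQS query context for the Puppeteer boss that supplies a mask's player target as the single context actor.
; ProvideSingleActor casts the querier to BGA_PuppeteerMask and returns its PlayerTarget when valid.
; Only a failed cast falls back to the first FirePetalBoss_PlayerTargetEQSDummy found in the world.
; A mask with an invalid target gets no fallback.
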(fn ProvideSingleActor
  (params (QuerierObject :type Object%20Reference) (QuerierActor :type Actor%20Object%20Reference))
  (bind _asbga_puppeteer_mask (Utilities|Casting|CastToBGA_PuppeteerMask QuerierActor)
    (:then
      (Utilities|IsValid (Variables|/PuppeteerBoss/Mask/BGA_PuppeteerMask.BGA_PuppeteerMask_C|GetPlayerTarget _asbga_puppeteer_mask)
        (:Is%20Valid
          (return (Variables|/PuppeteerBoss/Mask/BGA_PuppeteerMask.BGA_PuppeteerMask_C|GetPlayerTarget _asbga_puppeteer_mask)))))
    (:CastFailed
      (bind _returnvalue (Actor|GetActorOfClass "/FirePetalBosses/EQS/FirePetalBoss_PlayerTargetEQSDummy.FirePetalBoss_PlayerTargetEQSDummy_C"))
      (Utilities|IsValid _returnvalue
        (:Is%20Valid
          (return _returnvalue))))))
