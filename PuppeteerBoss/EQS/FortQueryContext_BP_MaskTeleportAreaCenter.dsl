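; EQS query context for the Puppeteer boss that supplies the center of the mask teleport area as a single location.
; If the querier is a PuppeteerMask, it returns the location of that mask's TeleportAreaBox, provided the box is valid.
; When the cast fails, it falls back to the location of a MaskTeleportArea_Box actor found in the world by class, if valid.
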(fn ProvideSingleLocation
  (params (QuerierObject :type Object%20Reference) (QuerierActor :type Actor%20Object%20Reference))
  (bind _asbga_puppeteer_mask (Utilities|Casting|CastToBGA_PuppeteerMask QuerierActor)
    (:then
      (Utilities|IsValid (Variables|/PuppeteerBoss/Mask/BGA_PuppeteerMask.BGA_PuppeteerMask_C|GetTeleportAreaBox _asbga_puppeteer_mask)
        (:Is%20Valid
          (return (Transformation|GetActorLocation (Variables|/PuppeteerBoss/Mask/BGA_PuppeteerMask.BGA_PuppeteerMask_C|GetTeleportAreaBox _asbga_puppeteer_mask))))))
    (:CastFailed
      (bind _returnvalue (Actor|GetActorOfClass "/PuppeteerBoss/Box/Puppeteer_MaskTeleportArea_Box.Puppeteer_MaskTeleportArea_Box_C"))
      (Utilities|IsValid _returnvalue
        (:Is%20Valid
          (return (Transformation|GetActorLocation _returnvalue)))))))
